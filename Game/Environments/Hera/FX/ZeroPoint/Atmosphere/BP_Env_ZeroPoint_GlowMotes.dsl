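; Blueprint actor for Zero Point glow motes, an atmospheric environment effect built on the SM_Motes_Few_CubeTown mesh component.
; The construction script creates a dynamic material instance from the mesh's first material slot and feeds the Overall Direction variable into its OverallDirection vector parameter.
; BeginPlay, Tick and overlap events are empty, so the Blueprint adds no runtime logic.
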(event EventBeginPlay)

(event Collision|EventActorBeginOverlap
  (params (OtherActor :type Actor%20Object%20Reference)))

(event EventTick
  (params (DeltaSeconds :type Float%20%28single-precision%29)))

(fn ConstructionScript
  (bind _returnvalue (Rendering|Material|CreateDynamicMaterialInstance (Variables|Default|GetSM_Motes_Few_CubeTown) 0 (Rendering|Material|GetMaterial (Variables|Default|GetSM_Motes_Few_CubeTown) 0) "None" :!class /Script/Engine.PrimitiveComponent))
  (Rendering|Material|SetVectorParameterValue _returnvalue "OverallDirection" (Math|Conversions|ToLinearColor(Vector) (Variables|Default|GetOverall%20Direction)) :!class /Script/Engine.MaterialInstanceDynamic))
